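; Blueprint vector function library for the Impostor Baker that converts between 3D directions and octahedral 2D coordinates.
; OctahedronToVector rebuilds Z from the absolute X and Y, unfolds the lower hemisphere where Z is negative, and normalizes.
; VectorToOctahedron projects by the L1 norm, folds non-positive Z, and returns those coordinates.
; It also returns a single float packing both coordinates, remapped to 0–1 and quantized in sixteenths.
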(fn OctahedronToVector
  (params (Octahedron :type Vector%202D%20Structure))
  (bind (_x _y) (Math|Vector2D|BreakVector2D Octahedron))
  (bind (_x_2 _y_2) (Math|Vector2D|BreakVector2D Octahedron))
  (bind (_x_3 _y_3) (Math|Vector2D|BreakVector2D Octahedron))
  (return (select (< (- 1 (Math|Vector2D|DotProduct(2D) (Math|Vector2D|MakeVector2D (Math|Float|Absolute(Float) _x_2) (Math|Float|Absolute(Float) _y_2)) "(X=1.000000,Y=1.000000)")) 0.0) (Math|Vector|Normalize (Math|Vector|MakeVector (* (Math|Float|SelectFloat 1 -1 (>= _x_3 0.0)) (- 1 (Math|Float|Absolute(Float) _y_3))) (* (Math|Float|SelectFloat 1 -1 (>= _y_3 0.0)) (- 1 (Math|Float|Absolute(Float) _x_3))) (- 1 (Math|Vector2D|DotProduct(2D) (Math|Vector2D|MakeVector2D (Math|Float|Absolute(Float) _x_2) (Math|Float|Absolute(Float) _y_2)) "(X=1.000000,Y=1.000000)")))) (Math|Vector|Normalize (Math|Vector|MakeVector _x _y (- 1 (Math|Vector2D|DotProduct(2D) (Math|Vector2D|MakeVector2D (Math|Float|Absolute(Float) _x_2) (Math|Float|Absolute(Float) _y_2)) "(X=1.000000,Y=1.000000)")))))))

(fn VectorToOctahedron
  (params (Vector :type Vector))
  (bind (_x _y) (Math|Vector2D|BreakVector2D (* (+ (select (<= (.z Vector) 0.0) (Math|Vector2D|MakeVector2D (* (Math|Float|SelectFloat 1 -1 (>= (Math|Vector2D|BreakVector2D (LocalVariables|Getxy)) 0.0)) (- 1 (Math|Float|Absolute(Float) (.Y (Math|Vector2D|BreakVector2D (/ (Math|Vector2D|MakeVector2D (.x Vector) (.y Vector)) (Math|Vector|DotProduct (Math|Vector|MakeVector (Math|Float|Absolute(Float) (.x Vector)) (Math|Float|Absolute(Float) (.y Vector)) (Math|Float|Absolute(Float) (.z Vector))) "1.000000,1.000000,1.000000"))))))) (* (Math|Float|SelectFloat 1 -1 (>= (.Y (Math|Vector2D|BreakVector2D (LocalVariables|Getxy))) 0.0)) (- 1 (Math|Float|Absolute(Float) (Math|Vector2D|BreakVector2D (/ (Math|Vector2D|MakeVector2D (.x Vector) (.y Vector)) (Math|Vector|DotProduct (Math|Vector|MakeVector (Math|Float|Absolute(Float) (.x Vector)) (Math|Float|Absolute(Float) (.y Vector)) (Math|Float|Absolute(Float) (.z Vector))) "1.000000,1.000000,1.000000"))))))) (/ (Math|Vector2D|MakeVector2D (.x Vector) (.y Vector)) (Math|Vector|DotProduct (Math|Vector|MakeVector (Math|Float|Absolute(Float) (.x Vector)) (Math|Float|Absolute(Float) (.y Vector)) (Math|Float|Absolute(Float) (.z Vector))) "1.000000,1.000000,1.000000"))) 1) 0.5)))
  (bind _x_2 (.x Vector))
  (bind _y_2 (.y Vector))
  (bind _z (.z Vector))
  (bind _returnvalue (Math|Float|Absolute(Float) _z))
  (bind _returnvalue_2 (Math|Float|Absolute(Float) _y_2))
  (bind _returnvalue_3 (Math|Float|Absolute(Float) _x_2))
  (bind _returnvalue_4 (Math|Vector|MakeVector _returnvalue_3 _returnvalue_2 _returnvalue))
  (bind _returnvalue_5 (Math|Vector|DotProduct _returnvalue_4 "1.000000,1.000000,1.000000"))
  (bind _y_3 (.y Vector))
  (bind _x_3 (.x Vector))
  (bind _returnvalue_6 (Math|Vector2D|MakeVector2D _x_3 _y_3))
  (bind _returnvalue_7 (/ _returnvalue_6 _returnvalue_5))
  (bind (_x_4 _y_4) (Math|Vector2D|BreakVector2D _returnvalue_7))
  (bind (_x_5 _y_5) (Math|Vector2D|BreakVector2D (LocalVariables|Getxy)))
  (LocalVariables|Setxy _returnvalue_7)
  (return (select (<= (.z Vector) 0.0) (Math|Vector2D|MakeVector2D (* (Math|Float|SelectFloat 1 -1 (>= _x_5 0.0)) (- 1 (Math|Float|Absolute(Float) _y_4))) (* (Math|Float|SelectFloat 1 -1 (>= _y_5 0.0)) (- 1 (Math|Float|Absolute(Float) _x_4)))) _returnvalue_7) (+ (/ (Math|Conversions|ToFloat(Integer) (Math|Float|Truncate (* _x 16))) 255) (/ (Math|Conversions|ToFloat(Integer) (Math|Float|Truncate (* _y 16))) 16))))
